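# Niagara vehicle boost cosmetic for S21 RLCS that draws ribbon trails behind the car's boost position.
# RibbonTrail builds the main gold-orange trail with animated glow and width, spawning at a quality-scaled rate.
# RibbonTrail001 adds thinner gold-to-silver strands, computing two phase-offset sine/cosine positions from user wave parameters, mirrored by spawn group.
# Game mode switches lifetime, backward velocity and strand width; painted-variant overrides are disabled.
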NiagaraSystem /VehicleCosmetics/Boosts/Boost_S21RLCS/Fx/NS_BoostS21RLCS
dependencies {
 usedDataInterfaces [
  &NiagaraDataInterfaceCurve
  &NiagaraDataInterfaceArrayFloat
  &NiagaraDataInterfaceColorCurve
  &NiagaraDataInterfaceSkeletalMesh
 ]
 usedDynamicInputs []
 usedModules [
  &self:SystemSpawnScript.SystemScriptSource.SystemScriptGraph.NiagaraNodeAssignment_0.SetVariables_#1
  &SystemState
 ]
}
emitters {
 RibbonTrail {
  emitterUpdateScript [
   &EmitterState(Life Cycle Mode=::NewEnumerator1, Loop Duration=1, UseLoopDelay=false)
   &VC_FilteredSocketCount()
   &VC_RateQualityHandler(High Spawn Rate=60, Low Spawn Rate=30, Use Custom Rate?=true)
   &VC_SpawnRibbon_v1(SpawnRate=$Emitter.Spawn Rate)
  ]
  particleSpawnScript [
   &VC_PaintedVariantsOverride() {enabled=false}
   &VC_InitializeRibbon_v1(
    Custom Offset=true
    Lifetime=&MergeHelper_FloatSwitch(0=1, 1=0.9, 2=1.1, 3=?, Index=$System.GameModeID, Num=::NewEnumerator1)
    Local Vector Offset=&AddVectorToPosition(
     Position=<0 0 0>
     Vector=&AddVectorToPosition(Position=$User.DTCarOffset, Vector=$User.SingleBoostPos)
    )
    Offset Amount=1
    Ribbon Width=103.8873
   )
   &AddVelocity(Velocity=&MergeHelper_VectorSwitch(0=<-1500 0 0>, 1=<-1500 0 0>, 2=<-600 0 0>, 3=?, Index=$System.GameModeID, Num=::NewEnumerator1)) {module=AddVelocity001}
  ]
  particleUpdateScript [
   &ParticleState()
   &InheritSourceMovement(ApplyForce=false, ApplyPosition=true, ApplyPositionFalloff=false)
   &SolveForcesAndVelocity(Clamp Velocity=false, Limit Acceleration=false)
   /SetVariables(
    Particles.GlowColor=&MultiplyLinearColor(
     Linear Color=<rgba 0.7231 0.247 0 1>
     Scale Factor=&MakeVector4FromVectorAndFloat(
      "Vector (RGB)"=<1 1 1>
      "Float (Alpha)"=&FloatFromCurve(FloatCurve=DataInterface() {propertyValues {Curve {keys [<at 0 1>,<at 1 0>]}}})
     )
    )
    Particles.GlowWidth:NiagaraFloat=&FloatFromCurve(
     FloatCurve=DataInterface() {propertyValues {Curve {
        defaultValue=1
        keys [
         {interpMode=RCIM_Cubic, tangentMode=RCTM_Break, time=0, value=1}
         {arriveTangent=-0.5, interpMode=RCIM_Cubic, leaveTangent=-0.5, time=0.4876, value=1}
         {interpMode=RCIM_Cubic, tangentMode=RCTM_Break, time=1, value=0.5}
        ]
       }}}
     Scale Curve=60
    )
    Particles.InnerColor:LinearColor=&ColorFromCurve(ColorCurve=DataInterface() {propertyValues {
       AlphaCurve {keys [<at 0 1 RCIM_Cubic>,<at 0.5724 20>,<at 1 2 RCIM_Cubic>]}
       BlueCurve {keys [<at 0.0263 0 RCIM_Cubic>,<at 0.5584 0>,<at 0.9365 0 RCIM_Cubic>]}
       GreenCurve {keys [<at 0.0263 0.0185 RCIM_Cubic>,<at 0.5584 0.589>,<at 0.9365 0.026 RCIM_Cubic>]}
       RedCurve {keys [<at 0.0263 0.0723 RCIM_Cubic>,<at 0.5584 0.802>,<at 0.9365 0.094 RCIM_Cubic>]}
      }})
    Particles.MainWidth:NiagaraFloat=&FloatFromCurve(
     FloatCurve=DataInterface() {propertyValues {Curve {
        defaultValue=1
        keys [
         {interpMode=RCIM_Cubic, tangentMode=RCTM_Break, time=-0.0047, value=0.15}
         {arriveTangent=-0.0253, interpMode=RCIM_Cubic, leaveTangent=-0.0253, tangentMode=RCTM_User, time=1, value=0.4}
        ]
       }}}
     Scale Curve=160
    )
   ) {isSetParameters=true}
   &VC_UpdateRibbon_v1(Enable Skeletal Position=true)
  ]
  rendererClasses [&NiagaraRibbonRendererProperties]
  renderers [
   {rendererClass=&NiagaraRibbonRendererProperties}
   {rendererClass=&NiagaraRibbonRendererProperties, rendererIndex=1}
  ]
 }
 RibbonTrail001 {
  emitterUpdateScript [
   &EmitterState(Life Cycle Mode=::NewEnumerator1, Loop Duration=1, UseLoopDelay=false)
   &VC_FilteredSocketCount()
   &VC_RateQualityHandler(Epic Spawn Rate=90, High Spawn Rate=90, Low Spawn Rate=30, Use Custom Rate?=true)
   &VC_SpawnRibbon_v1(Spawn Time=-1, SpawnRate=120)
  ]
  particleSpawnScript [
   &VC_PaintedVariantsOverride() {enabled=false}
   &VC_InitializeRibbon_v1(
    Custom Offset=true
    Lifetime=&MergeHelper_FloatSwitch(0=1.1, 1=0.9, 2=1.2, 3=?, Index=$System.GameModeID, Num=::NewEnumerator1)
    Local Vector Offset=&AddVectorToPosition(
     Position=<0 0 0>
     Vector=&AddVectorToPosition(Position=$User.DTCarOffset, Vector=$User.SingleBoostPos)
    )
    Offset Amount=1
    Ribbon Width=32
   )
   &AddVelocity(Velocity=&MergeHelper_VectorSwitch(0=<-1500 0 0>, 1=<-1500 0 0>, 2=<-600 0 0>, 3=?, Index=$System.GameModeID, Num=::NewEnumerator1)) {module=AddVelocity001}
  ]
  particleUpdateScript [
   &ParticleState()
   &InheritSourceMovement(ApplyForce=false, ApplyPosition=true, ApplyPositionFalloff=false)
   &SolveForcesAndVelocity(Clamp Velocity=false, Limit Acceleration=false)
   &ScaleRibbonWidth(Ribbon Width Scale=&MergeHelper_FloatSwitch(
     0=&FloatFromCurve(
      FloatCurve=DataInterface() {propertyValues {Curve {
         defaultValue=1
         keys [
          {arriveTangent=16.1741, interpMode=RCIM_Cubic, leaveTangent=16.1741, tangentMode=RCTM_User, time=0, value=-0.1}
          {arriveTangent=0.0438, leaveTangent=4.1708, tangentMode=RCTM_User, time=0.3, value=2}
          {
           arriveTangent=-0.8323
           arriveTangentWeight=0.2141
           interpMode=RCIM_Cubic
           leaveTangent=-0.8323
           leaveTangentWeight=0.2141
           tangentMode=RCTM_User
           time=0.7
           value=2
          }
          {arriveTangent=-0.348, interpMode=RCIM_Cubic, leaveTangent=-0.348, tangentMode=RCTM_User, time=0.95, value=0}
         ]
        }}}
      Scale Curve=1.2
     )
     1=&FloatFromCurve(
      FloatCurve=DataInterface() {propertyValues {Curve {
         defaultValue=1
         keys [
          {arriveTangent=16.1741, interpMode=RCIM_Cubic, leaveTangent=16.1741, tangentMode=RCTM_User, time=0, value=-0.1}
          {arriveTangent=0.0438, leaveTangent=4.1708, tangentMode=RCTM_User, time=0.3, value=2}
          {
           arriveTangent=-0.8323
           arriveTangentWeight=0.2141
           interpMode=RCIM_Cubic
           leaveTangent=-0.8323
           leaveTangentWeight=0.2141
           tangentMode=RCTM_User
           time=0.7
           value=2
          }
          {arriveTangent=-0.348, interpMode=RCIM_Cubic, leaveTangent=-0.348, tangentMode=RCTM_User, time=0.95, value=0}
         ]
        }}}
      Scale Curve=1.2
     )
     2=&FloatFromCurve(
      FloatCurve=DataInterface() {propertyValues {Curve {
         defaultValue=1
         keys [
          {arriveTangent=16.1741, interpMode=RCIM_Cubic, leaveTangent=16.1741, tangentMode=RCTM_User, time=0, value=-0.1}
          {arriveTangent=0.0438, leaveTangent=4.1708, tangentMode=RCTM_User, time=0.3, value=2}
          {
           arriveTangent=-0.8323
           arriveTangentWeight=0.2141
           interpMode=RCIM_Cubic
           leaveTangent=-0.8323
           leaveTangentWeight=0.2141
           tangentMode=RCTM_User
           time=0.7
           value=2
          }
          {arriveTangent=-0.348, interpMode=RCIM_Cubic, leaveTangent=-0.348, tangentMode=RCTM_User, time=0.95, value=0}
         ]
        }}}
      Scale Curve=1.5
     )
     3=?
     Index=$System.GameModeID
     Num=::NewEnumerator1
    ))
   &VC_UpdateRibbon_v1(Enable Skeletal Position=true)
   /SetVariables(
    Particles.CustomPosition_R1:NiagaraPosition=&AddVectorToPosition(
     Position=$Particles.Position
     Vector=&Multiply_VectorByFloat(
      Vector=&MakeVector(
       X=&Sine(
        Normalized Angle=&Add_Float(A=0.5, B=&Subtract_Float(A=$Emitter.NormalizedLoopAge, B=$Particles.NormalizedAge))
        Period=$User.WavePeriod
        Scale=&Multiply_Float(
         A=&SelectFloatFromArray(Direct Array Index=$Particles.SpawnGroup, Float Selection Array=DataInterface() {propertyValues {FloatData [1,-1]}})
         B=$User.WaveMaxScale
        )
        Bias=0
       )
       Y=&Sine(
        Normalized Angle=&Add_Float(A=0.5, B=&Subtract_Float(A=$Emitter.NormalizedLoopAge, B=$Particles.NormalizedAge))
        Period=$User.WavePeriod
        Scale=&Multiply_Float(
         A=&SelectFloatFromArray(Direct Array Index=$Particles.SpawnGroup, Float Selection Array=DataInterface() {propertyValues {FloatData [1,-1]}})
         B=$User.WaveMaxScale
        )
        Bias=0
       )
       Z=&Cosine(
        Normalized Angle=&Add_Float(A=0, B=&Subtract_Float(A=$Emitter.NormalizedLoopAge, B=$Particles.NormalizedAge))
        Period=$User.WavePeriod
        Scale=$User.WaveMaxScale
        Bias=0
       )
      )
      Float=$Particles.WaveScaleCurve
     )
    )
    Particles.CustomPosition_R2:NiagaraPosition=&AddVectorToPosition(
     Position=$Particles.Position
     Vector=&Multiply_VectorByFloat(
      Vector=&MakeVector(
       X=&Sine(
        Normalized Angle=&Add_Float(
         A=&Multiply_Float(A=$User.WavePeriod, B=0.5)
         B=&Subtract_Float(A=$Emitter.NormalizedLoopAge, B=$Particles.NormalizedAge)
        )
        Period=$User.WavePeriod
        Scale=&Multiply_Float(
         A=&SelectFloatFromArray(Direct Array Index=$Particles.SpawnGroup, Float Selection Array=DataInterface() {propertyValues {FloatData [-1,1]}})
         B=$User.WaveMaxScale
        )
        Bias=0
       )
       Y=&Sine(
        Normalized Angle=&Add_Float(
         A=&Multiply_Float(A=$User.WavePeriod, B=0.5)
         B=&Subtract_Float(A=$Emitter.NormalizedLoopAge, B=$Particles.NormalizedAge)
        )
        Period=$User.WavePeriod
        Scale=&Multiply_Float(
         A=&SelectFloatFromArray(Direct Array Index=$Particles.SpawnGroup, Float Selection Array=DataInterface() {propertyValues {FloatData [-1,1]}})
         B=$User.WaveMaxScale
        )
        Bias=0
       )
       Z=&Cosine(
        Normalized Angle=&Add_Float(
         A=&Multiply_Float(A=$User.WavePeriod, B=0.5)
         B=&Subtract_Float(A=$Emitter.NormalizedLoopAge, B=$Particles.NormalizedAge)
        )
        Period=$User.WavePeriod
        Scale=$User.WaveMaxScale
        Bias=0
       )
      )
      Float=$Particles.WaveScaleCurve
     )
    )
   ) {isSetParameters=true}
   /SetVariables(
    Particles.SilverColor:LinearColor=&ColorFromCurve(ColorCurve=DataInterface() {propertyValues {
       AlphaCurve {keys [<at 0.9136 100>,<at 1 0 RCIM_Cubic>]}
       BlueCurve {keys [<at 0.1027 0.25 RCIM_Cubic>,<at 0.173 0.422 RCIM_Cubic>]}
       GreenCurve {keys [<at 0.1027 0.4626 RCIM_Cubic>,<at 0.173 0.422 RCIM_Cubic>]}
       RedCurve {keys [<at 0.1027 1 RCIM_Cubic>,<at 0.173 0.422 RCIM_Cubic>]}
      }})
    Particles.WaveScaleCurve:NiagaraFloat=&FloatFromCurve(
     FloatCurve=DataInterface() {propertyValues {Curve {
        defaultValue=0
        keys [
         {
          arriveTangent=14.343
          arriveTangentWeight=1.6166
          interpMode=RCIM_Cubic
          leaveTangent=14.343
          leaveTangentWeight=1.6166
          tangentMode=RCTM_User
          tangentWeightMode=RCTWM_WeightedBoth
          time=0
          value=0
         }
         {arriveTangent=0.0973, interpMode=RCIM_Cubic, leaveTangent=1.7653, tangentMode=RCTM_Break, time=0.8756, value=1.4294}
         {arriveTangent=-16.3188, tangentMode=RCTM_User, time=1, value=0}
        ]
       }}}
     Scale Curve=1.2
    )
   ) {isSetParameters=true}
  ]
  rendererClasses [&NiagaraRibbonRendererProperties]
  renderers [
   {rendererClass=&NiagaraRibbonRendererProperties}
   {rendererClass=&NiagaraRibbonRendererProperties, rendererIndex=1}
  ]
 }
}
summary {
 systemName=NS_BoostS21RLCS
 emitters [
  {emitterName=RibbonTrail001, rendererClasses [&NiagaraRibbonRendererProperties]}
  {emitterName=RibbonTrail, rendererClasses [&NiagaraRibbonRendererProperties]}
 ]
 userVariables [
  User.bIsFrontendPreview {type=NiagaraBool}
  User.bIsPainted {type=NiagaraBool}
  User.ColorVariantID {type=NiagaraInt32}
  User.DTCarOffset {type=NiagaraPosition, description="", defaultValue {}}
  User.PaintedColor {type=LinearColor, description="", defaultValue {struct=LinearColor, value=<rgba 1 1 1 1>}}
  User.PaintedScalar {type=NiagaraFloat, defaultValue=1, description=""}
  User.PortCount {type=NiagaraInt32, defaultValue=1}
  User.SingleBoostOffset {type=Vector3f, description="", defaultValue {struct=Vector3f, value=<0 0 0>}}
  User.SingleBoostPos {type=Vector3f, description="", defaultValue {struct=Vector3f, value=<0 0 0>}}
  User.WaveMaxScale {type=NiagaraFloat, defaultValue=10, description=""}
  User.WavePeriod {type=NiagaraFloat, defaultValue=1, description=""}
 ]
}
